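# Physics asset for the Futura vehicle body mesh, supplying its collision and physics shape.
# It holds a single box body on the root bone, offset forward and raised to enclose the whole car body.
# It defines no constraints, so the body acts as one rigid shape.
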PhysicsAsset /VehicleCosmetics/Bodies/Futura/Meshes/Body_Futura_PhysicsDM
bodies {root [{shapeType=Box, extentX=444, extentY=216, extentZ=116, shapeName=root_box, center=<34.7151 0 99.1859>}]}
constraints []
